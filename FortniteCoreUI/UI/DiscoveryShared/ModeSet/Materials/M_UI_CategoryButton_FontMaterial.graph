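# M_UI_CategoryButton_FontMaterial is a UI-domain material that sets the font color for Discovery mode set category buttons.
# It nests two lerps over three color parameters: default white, highlight dark navy and warning red.
# AnimateWarning blends highlight toward warning, and AnimateHighlight blends default toward that result, so warning only shows as far as highlight is raised.
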Material /FortniteCoreUI/UI/DiscoveryShared/ModeSet/Materials/M_UI_CategoryButton_FontMaterial
graph {
 properties {bUsedWithStaticMesh=false, materialDomain=MD_UI, pixelDepthOffsetMode=PDOM_Legacy, usedShadingModels=Default Lit}
 nodes [LinearInterpolate(
   A=VectorParameter() {parameterName=RGB_Default, defaultValue=<rgba 1 1 1 1>}.RGB
   B=LinearInterpolate(
    A=VectorParameter() {parameterName=RGB_Highlight, defaultValue=<rgba 0 0.004 0.0723 1>}.RGB
    B=VectorParameter() {parameterName=RGB_Warning, defaultValue=<rgba 0.949 0.0549 0.1255 1>}.RGB
    Alpha=ScalarParameter() {parameterName=AnimateWarning}
   )
   Alpha=ScalarParameter() {parameterName=AnimateHighlight}
  )]
}
