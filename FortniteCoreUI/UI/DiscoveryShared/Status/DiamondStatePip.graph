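# Diamond-shaped status pip widget for shared Discovery UI, a CommonUserWidget overlaying a filled image and an outline image.
# Construct calls ResetState, which makes the fill transparent and hides the outline; SetHoverState toggles outline visibility.
# SetSelectedState plays the Selection animation forward, or in reverse at double speed when deselected.
# That animation fades the fill in from blue to white with a scale pop while the outline shrinks, rebounds and fades out.
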WidgetBlueprint /FortniteCoreUI/UI/DiscoveryShared/Status/DiamondStatePip
animations [Selection {
  binding_count=2
  length_seconds=0.3
  object=&self:Selection
  sequence {bindings [
    Pip {
     id=#1
     possessed_class=Image
     tracks [
      {
       type=MovieSceneColorTrack
       display_name=Color and Opacity
       sections [{
         type=MovieSceneColorSection
         channels [
          Color.R {type=MovieSceneScriptingFloatChannel, keys [[0,0.0242],[2.9997,1]]}
          Color.G {type=MovieSceneScriptingFloatChannel, keys [[0,0.5906],[2.9997,1]]}
          Color.B {type=MovieSceneScriptingFloatChannel, keys [[0,0.8148],[2.9997,1]]}
          Color.A {type=MovieSceneScriptingFloatChannel, const [1,2]}
         ]
        }]
      }
      {
       type=MovieSceneFloatTrack
       display_name=Render Opacity
       sections [{type=MovieSceneFloatSection, channels [None {type=MovieSceneScriptingFloatChannel, keys [[0,0],[2.9997,1]]}]}]
      }
      {
       type=MovieScene2DTransformTrack
       sections [{
         type=MovieScene2DTransformSection
         channels [
          Scale.X {type=MovieSceneScriptingFloatChannel, keys [[0,0.75],[4,1.5],[6,1]]}
          Scale.Y {type=MovieSceneScriptingFloatChannel, keys [[0,0.75],[4,1.5],[6,1]]}
         ]
        }]
      }
     ]
    }
    Pip_Outline {
     id=#2
     possessed_class=Image
     tracks [
      {
       type=MovieSceneFloatTrack
       display_name=Render Opacity
       sections [{type=MovieSceneFloatSection, channels [None {type=MovieSceneScriptingFloatChannel, keys [[0,1],[2.9997,0]]}]}]
      }
      {
       type=MovieScene2DTransformTrack
       sections [{
         type=MovieScene2DTransformSection
         channels [
          Scale.X {type=MovieSceneScriptingFloatChannel, keys [[0,1],[2,0.5],[2.9997,1]]}
          Scale.Y {type=MovieSceneScriptingFloatChannel, keys [[0,1],[2,0.5],[2.9997,1]]}
         ]
        }]
      }
     ]
    }
   ]}
 }]
graphs [
 EventGraph {nodes [
   n1 = SetHoverState!() -> (OutputDelegate, then, IsHovered:bool=~) {is=title K2Node_CustomEvent}
   > SetVisibility(
    self=Get Pip_Outline() -> (Pip_Outline:Image Object Reference)
    InVisibility=Select(Option 0=Hidden, Option 1=SelfHitTestInvisible, Index=@n1.IsHovered)
   )
   n2 = SetSelectedState!() -> (OutputDelegate, then, IsSelected:bool=~) {is=title K2Node_CustomEvent}
   > n3 = Branch(Condition=@n2.IsSelected)
   n4 = Get Selection() -> (Selection:Widget Animation Object Reference)
   > PlayAnimationForward(InAnimation=@n4)
   PlayAnimationReverse(execute=@n3.else, InAnimation=@n4, PlaybackSpeed=2)
   Event Construct()
   > Parent: Construct()
   > ResetState()
  ]}
 ResetState {nodes [
   ResetState() {is=title K2Node_FunctionEntry}
   > SetRenderOpacity(self=Get Pip() -> (Pip:Image Object Reference), InOpacity=0)
   > SetVisibility(self=Get Pip_Outline(), InVisibility=Hidden)
  ]}
]
variables [
 Is Hovered {type=boolean}
 Is Selected {type=boolean}
]
widgets {
 info {inheritedWidgetCount=0, parentClass=&CommonUserWidget, rootWidgetClass=&Overlay, widgetCount=3}
 widgets [
  Overlay_1 {widgetClassPath=&Overlay, widget=Overlay_1}
  Pip_Outline {
   widgetClassPath=&Image
   bIsVariable=true
   parent=Overlay_1
   widget=Pip_Outline
   properties {
    visibility=HitTestInvisible
    brush {
     bIsDynamicallyLoaded=false
     drawAs=Image
     imageType=NoImage
     mirroring=NoMirror
     resourceName=None
     resourceObject=/FortniteCoreUI/UI/DiscoveryShared/Status/Materials/MI_UI_DiamondStatePip_Outline
     tiling=NoTile
     imageSize=<26 26>
     margin=<ltrb 0 0 0 0>
     outlineSettings {
      bUseBrushTransparency=false
      roundingType=HalfHeightRadius
      width=0
      color {colorUseRule=UseColor_Specified, specifiedColor=<rgba 0 0 0 0>}
      cornerRadii=<0 0 0 0>
     }
     tintColor {colorUseRule=UseColor_Specified, specifiedColor=<rgba 1 1 1 1>}
     uVRegion {bIsValid=false, max=<0 0>, min=<0 0>}
    }
    colorAndOpacity=<rgba 0.8148 0.8148 0.8148 1>
   }
   slot {class=OverlaySlot, properties {horizontalAlignment=HAlign_Fill, verticalAlignment=VAlign_Fill, padding=<ltrb -4 -4 -4 -4>}}
  }
  Pip {
   widgetClassPath=&Image
   bIsVariable=true
   parent=Overlay_1
   widget=Pip
   properties {
    visibility=HitTestInvisible
    brush {
     bIsDynamicallyLoaded=false
     drawAs=Image
     imageType=NoImage
     mirroring=NoMirror
     resourceName=None
     resourceObject=/FortniteCoreUI/UI/DiscoveryShared/Status/Materials/MI_UI_DiamondStatePip
     tiling=NoTile
     imageSize=<26 26>
     margin=<ltrb 0 0 0 0>
     outlineSettings {
      bUseBrushTransparency=false
      roundingType=HalfHeightRadius
      width=0
      color {colorUseRule=UseColor_Specified, specifiedColor=<rgba 0 0 0 0>}
      cornerRadii=<0 0 0 0>
     }
     tintColor {colorUseRule=UseColor_Specified, specifiedColor=<rgba 1 1 1 1>}
     uVRegion {bIsValid=false, max=<0 0>, min=<0 0>}
    }
    colorAndOpacity=<rgba 0.8148 0.8148 0.8148 1>
   }
   slot {class=OverlaySlot, properties {horizontalAlignment=HAlign_Fill, verticalAlignment=VAlign_Fill}}
  }
 ]
}
