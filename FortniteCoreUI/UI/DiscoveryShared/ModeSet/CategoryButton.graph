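# Category button widget for the Discovery mode-set UI, built on CommonButtonBase with an icon slot, title, input-action prompt, diamond state pip and side-configurable safe zone.
# Hover and selection update the pip and play forward/reverse animations that drive background and title material parameters.
# SetWarningStatus swaps the pip for a warning icon and dims the button.
# While warned, hovering shows the warning message instead of the title, and clicking pulses the icon.
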WidgetBlueprint /FortniteCoreUI/UI/DiscoveryShared/ModeSet/CategoryButton
animations [
 OnHover {
  binding_count=4
  length_seconds=0.2
  object=&self:OnHover
  sequence {bindings [
    Image_Bg {
     id=#1
     possessed_class=Image
     tracks [{
       type=MovieSceneWidgetMaterialTrack
       sections [{
         type=MovieSceneParameterSection
         end_frame=4
         start_frame=0
         channels [AnimateHighlight {type=MovieSceneScriptingFloatChannel, keys [[0,0],[3,1],[4,1]]}]
        }]
      }]
    }
    DisabledIcon {
     id=#2
     possessed_class=SizeBox
     tracks [{
       type=MovieSceneFloatTrack
       display_name=Render Opacity
       sections [{type=MovieSceneFloatSection, channels [None {type=MovieSceneScriptingFloatChannel, keys [[0,0],[3,1]]}]}]
      }]
    }
    Text_Title {
     id=#3
     possessed_class=CommonTextBlock
     tracks [{
       type=MovieSceneWidgetMaterialTrack
       display_name=Font.FontMaterial
       sections [{
         type=MovieSceneParameterSection
         end_frame=4
         start_frame=0
         channels [AnimateHighlight {type=MovieSceneScriptingFloatChannel, keys [[0,0],[3,1]]}]
        }]
      }]
    }
    Border_PipColor {
     id=#4
     possessed_class=Border
     tracks [{
       type=MovieSceneColorTrack
       display_name=Content Color and Opacity
       sections [{
         type=MovieSceneColorSection
         channels [
          Color.R {type=MovieSceneScriptingFloatChannel, keys [[0,1],[3,0]]}
          Color.G {type=MovieSceneScriptingFloatChannel, keys [[0,1],[3,0.004]]}
          Color.B {type=MovieSceneScriptingFloatChannel, keys [[0,1],[3,0.0723]]}
          Color.A {type=MovieSceneScriptingFloatChannel}
         ]
        }]
      }]
    }
   ]}
 }
 OnWarning {
  binding_count=3
  length_seconds=0.15
  object=&self:OnWarning
  sequence {bindings [
    Image_Dim {
     id=#5
     possessed_class=Image
     tracks [{
       type=MovieSceneFloatTrack
       display_name=Render Opacity
       sections [{type=MovieSceneFloatSection, channels [None {type=MovieSceneScriptingFloatChannel, keys [[0,0],[3,1]]}]}]
      }]
    }
    Image_Bg {
     id=#1
     possessed_class=Image
     tracks [{
       type=MovieSceneWidgetMaterialTrack
       sections [{type=MovieSceneParameterSection, channels [AnimateWarning {type=MovieSceneScriptingFloatChannel, keys [[0,0],[3,1]]}]}]
      }]
    }
    Text_Title {
     id=#6
     possessed_class=CommonTextBlock
     tracks [{
       type=MovieSceneWidgetMaterialTrack
       display_name=Font.FontMaterial
       sections [{type=MovieSceneParameterSection, channels [AnimateWarning {type=MovieSceneScriptingFloatChannel, keys [[0,0],[3,1]]}]}]
      }]
    }
   ]}
 }
 OnSelect {
  binding_count=1
  length_seconds=0.15
  object=&self:OnSelect
  sequence {bindings [Image_Bg {
     id=#7
     possessed_class=Image
     tracks [{
       type=MovieSceneWidgetMaterialTrack
       sections [{type=MovieSceneParameterSection, channels [AnimateSelection {type=MovieSceneScriptingFloatChannel, keys [[0,0],[3,1]]}]}]
      }]
    }]}
 }
 Dim {
  binding_count=1
  length_seconds=0.15
  object=&self:Dim
  sequence {bindings [Image_Dim {
     id=#8
     possessed_class=Image
     tracks [{
       type=MovieSceneFloatTrack
       display_name=Render Opacity
       sections [{type=MovieSceneFloatSection, channels [None {type=MovieSceneScriptingFloatChannel, keys [[0,0],[3,1]]}]}]
      }]
    }]}
 }
 Reset {
  binding_count=3
  length_seconds=0.15
  object=&self:Reset
  sequence {bindings [
    Image_Bg {
     id=#1
     possessed_class=Image
     tracks [{
       type=MovieSceneWidgetMaterialTrack
       sections [{
         type=MovieSceneParameterSection
         channels [
          AnimateHighlight {type=MovieSceneScriptingFloatChannel, const [0,2]}
          AnimateSelection {type=MovieSceneScriptingFloatChannel, const [0,1]}
          AnimateWarning {type=MovieSceneScriptingFloatChannel, const [0,1]}
         ]
        }]
      }]
    }
    Text_Title {
     id=#9
     possessed_class=CommonTextBlock
     tracks [{
       type=MovieSceneWidgetMaterialTrack
       display_name=Font.FontMaterial
       sections [{
         type=MovieSceneParameterSection
         channels [
          AnimateHighlight {type=MovieSceneScriptingFloatChannel, const [0,1]}
          AnimateWarning {type=MovieSceneScriptingFloatChannel, const [0,1]}
         ]
        }]
      }]
    }
    Border_PipColor {
     id=#10
     possessed_class=Border
     tracks [{
       type=MovieSceneColorTrack
       display_name=Content Color and Opacity
       sections [{
         type=MovieSceneColorSection
         end_frame=3
         start_frame=0
         channels [
          Color.R {type=MovieSceneScriptingFloatChannel, const [1,1]}
          Color.G {type=MovieSceneScriptingFloatChannel, const [1,1]}
          Color.B {type=MovieSceneScriptingFloatChannel, const [1,1]}
          Color.A {type=MovieSceneScriptingFloatChannel}
         ]
        }]
      }]
    }
   ]}
 }
 PulseWarningIcon {
  binding_count=1
  length_seconds=0.44
  object=&self:PulseWarningIcon
  sequence {bindings [IconOvr {
     id=#11
     possessed_class=Overlay
     tracks [{
       type=MovieScene2DTransformTrack
       sections [{
         type=MovieScene2DTransformSection
         channels [
          Scale.X {type=MovieSceneScriptingFloatChannel, keys [[4,1],[5.6,1.25],[8.8,1]]}
          Scale.Y {type=MovieSceneScriptingFloatChannel, keys [[4,1],[5.6,1.25],[8.8,1]]}
         ]
        }]
      }]
    }]}
 }
]
graphs [
 EventGraph {nodes [
   "On Input Icon Updated (InputActionWidget)"() {is=title K2Node_ComponentBoundEvent}
   > n1 = GetIconDynamicMaterial!(execute, self:Common Action Widget Object Reference=Get InputActionWidget() -> (InputActionWidget:Common Action Widget Object Reference)) -> (then, ReturnValue:Material Instance Dynamic Object Reference)
   > SetScalarParameterValue(self=@n1, ParameterName=Use Dark, Value=1)
   > SetScalarParameterValue(self=@n1, ParameterName=Outline Opacity, Value=0)
   PlayClickedEvents() {is=title K2Node_CustomEvent}
   > Branch(Condition=Get IsWarningOn() -> (IsWarningOn:bool))
   > PlayAnimationForward(InAnimation=Get PulseWarningIcon() -> (PulseWarningIcon:Widget Animation Object Reference))
   n2 = SetHovered!() -> (OutputDelegate, then, IsHovered:bool=~) {is=title K2Node_CustomEvent}
   > n3 = Set Is Hovered(Is Hovered:bool=@n2.IsHovered) -> (Output_Get:bool)
   > SetHoverState!(
    execute
    self:Diamond State Pip Object Reference=Get DiamondStatePip() -> (DiamondStatePip:Diamond State Pip Object Reference)
    IsHovered:bool=@n3
   ) -> (then)
   > SetRenderOpacity(self=Get InputActionWidget(), InOpacity=Select(Option 0:double=0, Option 1:double=1, Index=@n3) -> (ReturnValue:double))
   > n4 = Branch(Condition=@n3)
   n5 = Get Text_Title() -> (Text_Title:Common Text Object Reference)
   > "SetText (Text)"(
    self=@n5
    InText=Select(
     Option 0:"Text (by ref)"=Get TitleText() -> (TitleText:Text)
     Option 1:"Text (by ref)"=Get Warning Message() -> (Warning Message:Text)
     Index=Get IsWarningOn()
    ) -> (ReturnValue:"Text (by ref)")
   )
   n6 = Get OnHover() -> (OnHover:Widget Animation Object Reference)
   > PlayAnimationForward(InAnimation=@n6)
   n7 = Get Image_Dim() -> (Image_Dim:Image Object Reference)
   > SetVisibility(self=@n7, InVisibility=Collapsed)
   "SetText (Text)"(execute=@n4.else, self=@n5, InText=Get TitleText())
   > PlayAnimationReverse(InAnimation=@n6)
   > SetVisibility(self=@n7, InVisibility=SelfHitTestInvisible)
   n8 = SetSelected!() -> (OutputDelegate, then, IsSelected:bool=~) {is=title K2Node_CustomEvent}
   > n9 = Set IsSelected(IsSelected:bool=@n8.IsSelected) -> (Output_Get:bool)
   > SetSelectedState!(execute, self:Diamond State Pip Object Reference=Get DiamondStatePip(), IsSelected:bool=@n9) -> (then)
   > n10 = Branch(Condition=@n9)
   n11 = Get OnSelect() -> (OnSelect:Widget Animation Object Reference)
   > PlayAnimationForward(InAnimation=@n11)
   PlayAnimationReverse(execute=@n10.else, InAnimation=@n11)
   Event On Selected()
   > SetSelected!(execute, self, IsSelected:bool=true) -> (then) {is=title K2Node_CallFunction}
   Event On Deselected()
   > SetSelected(IsSelected=false)
   Event On Hovered()
   > SetHovered!(execute, self, IsHovered:bool=true) -> (then) {is=title K2Node_CallFunction}
   Event On Clicked()
   > PlayClickedEvents() {is=title K2Node_CallFunction}
   Event PreConstruct()
   > SetSafeZonePadding!(execute, self, Pad Left:bool=Get Pad Left() -> (Pad Left:bool), Pad Right:bool=Get Pad Right() -> (Pad Right:bool)) -> (then)
   Event On Unhovered()
   > SetHovered(IsHovered=false)
   Event Construct()
   > Parent: Construct()
   > ResetState()
  ]}
 ResetState {nodes [
   ResetState() {is=title K2Node_FunctionEntry}
   > PlayAnimationForward(InAnimation=Get Reset() -> (Reset:Widget Animation Object Reference))
   > ResetState(self:Diamond State Pip Object Reference=Get DiamondStatePip()) {is=title K2Node_CallFunction}
   n1 = Get DisabledIcon() -> (DisabledIcon:Size Box Object Reference)
   > SetRenderOpacity(self=@n1, InOpacity=0)
   > SetVisibility(self=@n1, InVisibility=Collapsed)
   > SetRenderOpacity(
    self=Get InputActionWidget()
    InOpacity=Select(Option 0:double=0, Option 1:double=1, Index=Get Is Hovered() -> (Is Hovered:bool)) -> (ReturnValue:double)
   )
  ]}
 SetSafeZonePadding {nodes [
   n1 = SetSafeZonePadding!() -> (then, Pad Left:bool=~, Pad Right:bool=~) {is=title K2Node_FunctionEntry}
   > n2 = Set Pad Left(Pad Left:bool=@n1) -> (Output_Get:bool)
   > n3 = Set Pad Right(Pad Right:bool=@n1.Pad Right) -> (Output_Get:bool)
   > SetSidesToPad!(
    execute
    self:Safe Zone Object Reference=Get SafeZone() -> (SafeZone:Safe Zone Object Reference)
    InPadLeft:bool=@n2
    InPadRight:bool=@n3
    InPadTop:bool=false
    InPadBottom:bool=false
   ) -> (then)
  ]}
 SetTitleText {nodes [
   n1 = SetTitleText!() -> (then, TitleText:"Text (by ref)") {is=title K2Node_FunctionEntry}
   > n2 = Set TitleText(TitleText:Text=@n1) -> (Output_Get:Text)
   n3 = Get Is Hovered()
   n4 = Get IsWarningOn()
   > Branch(Condition=OR Boolean(A=NOT Boolean(A=@n4), B=NOT Boolean(A=@n3)))
   > "SetText (Text)"(self=Get Text_Title(), InText=@n2)
  ]}
 SetWarningStatus {nodes [
   n1 = SetWarningStatus!() -> (then, WarningOn:bool=~, WarningMessage:Text) {is=title K2Node_FunctionEntry}
   > Set PreviousIsWarningOn(PreviousIsWarningOn:bool=Get IsWarningOn()) -> (Output_Get:bool)
   > Set IsWarningOn(IsWarningOn:bool=@n1) -> (Output_Get:bool)
   > Set Warning Message(Warning Message:Text=@n1.WarningMessage) -> (Output_Get:Text)
   > SetVisibility(
    self=Get DiamondStatePip()
    InVisibility=Select(Option 0:ESlateVisibility Enum=SelfHitTestInvisible, Option 1:ESlateVisibility Enum=Collapsed, Index=Get IsWarningOn()) -> (ReturnValue:ESlateVisibility Enum)
   )
   > SetVisibility(self=Get DisabledIcon(), InVisibility=Select(Option 0=Collapsed, Option 1=SelfHitTestInvisible, Index=Get IsWarningOn()))
   > n2 = Sequence()
   > n3 = Branch(Condition=AND Boolean(A=Get IsWarningOn(), B=Get Is Hovered()))
   > "SetText (Text)"(self=Get Text_Title(), InText=Get Warning Message())
   "SetText (Text)"(execute=@n3.else, self=Get Text_Title(), InText=Get TitleText())
   Branch(
    execute=@n2.then_1
    Condition="Not Equal (Boolean)"(A=Get IsWarningOn(), B=Get PreviousIsWarningOn() -> (PreviousIsWarningOn:bool)) -> (ReturnValue=~)
   )
   > n4 = Branch(Condition=Get IsWarningOn())
   n5 = Get OnWarning() -> (OnWarning:Widget Animation Object Reference)
   > PlayAnimationForward(InAnimation=@n5)
   PlayAnimationReverse(execute=@n4.else, InAnimation=@n5)
  ]}
]
variables [
 Is Hovered {type=boolean}
 IsSelected {type=boolean}
 IsWarningOn {type=boolean}
 Pad Left {type=boolean}
 Pad Right {type=boolean}
 TitleText {type=string, default=""}
 Warning Message {type=string, default=""}
]
widgets {
 info {inheritedWidgetCount=1, parentClass=&CommonButtonBase, rootWidgetClass=&SizeBox, widgetCount=17}
 widgets [
  SizeBox_All {
   widgetClassPath=&SizeBox
   widget=SizeBox_All
   properties {bOverride_HeightOverride=true, bOverride_MinDesiredWidth=true, heightOverride=100, minDesiredWidth=625, widthOverride=625}
  }
  Overlay_21 {widgetClassPath=&Overlay, parent=SizeBox_All, widget=Overlay_21, slot {class=SizeBoxSlot}}
  Image_Bg {
   widgetClassPath=&Image
   parent=Overlay_21
   widget=Image_Bg
   properties {brush {
     bIsDynamicallyLoaded=false
     drawAs=Image
     imageType=NoImage
     mirroring=NoMirror
     resourceName=None
     resourceObject=/Game/UI/DiscoveryShared/ModeSet/Materials/MI_UI_SidebarStyle_Button_CategoryInst
     tiling=NoTile
     imageSize=<32 32>
     margin=<ltrb 0 0 0 0>
     outlineSettings {
      bUseBrushTransparency=false
      roundingType=HalfHeightRadius
      width=0
      color {colorUseRule=UseColor_Specified, specifiedColor=<rgba 0 0 0 0>}
      cornerRadii=<0 0 0 0>
     }
     tintColor {colorUseRule=UseColor_Specified, specifiedColor=<rgba 1 1 1 1>}
     uVRegion {bIsValid=false, max=<0 0>, min=<0 0>}
    }}
   slot {class=OverlaySlot, properties {horizontalAlignment=HAlign_Fill, verticalAlignment=VAlign_Fill}}
  }
  SafeZone {
   widgetClassPath=&SafeZone
   bIsVariable=true
   parent=Overlay_21
   widget=SafeZone
   properties {padBottom=false, padLeft=false, padTop=false}
   slot {class=OverlaySlot, properties {horizontalAlignment=HAlign_Fill, verticalAlignment=VAlign_Fill}}
  }
  HorizontalBox_0 {widgetClassPath=&HorizontalBox, parent=SafeZone, widget=HorizontalBox_0, slot {class=SafeZoneSlot}}
  SizeBox_Image {
   widgetClassPath=&SizeBox
   parent=HorizontalBox_0
   widget=SizeBox_Image
   properties {bOverride_WidthOverride=true, widthOverride=155}
   slot {class=HorizontalBoxSlot}
  }
  Overlay_2 {widgetClassPath=&Overlay, parent=SizeBox_Image, widget=Overlay_2, slot {class=SizeBoxSlot}}
  NamedSlot_Icon {
   widgetClassPath=&NamedSlot
   parent=Overlay_2
   widget=NamedSlot_Icon
   properties {bExposeOnInstanceOnly=true}
   slot {class=OverlaySlot}
  }
  DisabledIcon {
   widgetClassPath=&SizeBox
   bIsVariable=true
   parent=Overlay_2
   widget=DisabledIcon
   properties {bOverride_HeightOverride=true, bOverride_WidthOverride=true, heightOverride=40, widthOverride=40}
   slot {class=OverlaySlot, properties {horizontalAlignment=HAlign_Center, verticalAlignment=VAlign_Center}}
  }
  IconOvr {widgetClassPath=&Overlay, parent=DisabledIcon, widget=IconOvr, properties {visibility=HitTestInvisible}, slot {class=SizeBoxSlot}}
  WarningTriangle {
   widgetClassPath=&Image
   parent=IconOvr
   widget=WarningTriangle
   properties {
    visibility=SelfHitTestInvisible
    brush {
     bIsDynamicallyLoaded=false
     drawAs=Image
     imageType=NoImage
     mirroring=NoMirror
     resourceName=None
     resourceObject=/FortniteCoreUI/UI/DiscoveryShared/ModeSet/Materials/MI_UI_CategoryButton_WarningTriangle
     tiling=NoTile
     imageSize=<32 32>
     margin=<ltrb 0 0 0 0>
     outlineSettings {
      bUseBrushTransparency=false
      roundingType=HalfHeightRadius
      width=0
      color {colorUseRule=UseColor_Specified, specifiedColor=<rgba 0 0 0 0>}
      cornerRadii=<0 0 0 0>
     }
     tintColor {colorUseRule=UseColor_Specified, specifiedColor=<rgba 1 1 1 1>}
     uVRegion {bIsValid=false, max=<0 0>, min=<0 0>}
    }
   }
   slot {class=OverlaySlot, properties {horizontalAlignment=HAlign_Fill, verticalAlignment=VAlign_Fill}}
  }
  Exclamation {
   widgetClassPath=&CommonTextBlock
   parent=IconOvr
   widget=Exclamation
   properties {
    text="!"
    font {
     bForceMonospaced=false
     bMaterialIsStencil=false
     fontMaterial=null
     fontObject=&Burbank
     letterSpacing=0
     monospacedWidth=1
     size=18
     skewAmount=0
     typefaceFontName=BurbankRegular-Black
     outlineSettings {
      bApplyOutlineToDropShadows=false
      bMiteredCorners=false
      bSeparateFillAlpha=false
      outlineMaterial=null
      outlineSize=0
      outlineColor=<rgba 0 0 0 1>
     }
    }
    margin=<ltrb 0 -4 0 -4>
    renderTransform {angle=0, scale=<1 1>, shear=<4 0>, translation=<-1 5>}
    shadowOffset=<0 0>
   }
   slot {class=OverlaySlot, properties {horizontalAlignment=HAlign_Center, verticalAlignment=VAlign_Center}}
  }
  Text_Title {
   widgetClassPath=&CommonTextBlock
   bIsVariable=true
   parent=HorizontalBox_0
   widget=Text_Title
   properties {
    autoWrapText=true
    mobileFontSizeMultiplier=1.4
    text="{Duos}"
    font {
     bForceMonospaced=false
     bMaterialIsStencil=false
     fontMaterial=/FortniteCoreUI/UI/DiscoveryShared/ModeSet/Materials/MI_UI_CategoryButton_FontMaterial_Title
     fontObject=&BurbankSmall
     letterSpacing=0
     monospacedWidth=1
     size=16
     skewAmount=0
     typefaceFontName=Black
     outlineSettings {
      bApplyOutlineToDropShadows=false
      bMiteredCorners=false
      bSeparateFillAlpha=false
      outlineMaterial=null
      outlineSize=0
      outlineColor=<rgba 0 0 0 1>
     }
    }
    shadowOffset=<0 0>
   }
   slot {class=HorizontalBoxSlot, properties {verticalAlignment=VAlign_Center, padding=<ltrb 20 0 20 0>, size {sizeRule=Fill, value=1}}}
  }
  InputActionWidget {
   widgetClassPath=&CommonActionWidget
   bInherited=true
   bIsVariable=true
   parent=HorizontalBox_0
   widget=InputActionWidget
   properties {
    progressMaterialParam=Percentage
    progressMaterialBrush {
     bIsDynamicallyLoaded=false
     drawAs=Image
     imageType=NoImage
     mirroring=NoMirror
     resourceName=None
     resourceObject=/Game/UI/Materials/HollowCircularProgressBar
     tiling=NoTile
     imageSize=<36 36>
     margin=<ltrb 0 0 0 0>
     outlineSettings {
      bUseBrushTransparency=false
      roundingType=HalfHeightRadius
      width=0
      color {colorUseRule=UseColor_Specified, specifiedColor=<rgba 0 0 0 0>}
      cornerRadii=<0 0 0 0>
     }
     tintColor {colorUseRule=UseColor_Specified, specifiedColor=<rgba 1 1 0 1>}
     uVRegion {bIsValid=false, max=<0 0>, min=<0 0>}
    }
   }
   slot {class=HorizontalBoxSlot, properties {horizontalAlignment=HAlign_Center, verticalAlignment=VAlign_Center, padding=<ltrb 0 0 5 0>}}
  }
  Border_PipColor {
   widgetClassPath=&Border
   parent=HorizontalBox_0
   widget=Border_PipColor
   properties {background {
     bIsDynamicallyLoaded=false
     drawAs=NoDrawType
     imageType=NoImage
     mirroring=NoMirror
     resourceName=None
     resourceObject=null
     tiling=NoTile
     imageSize=<32 32>
     margin=<ltrb 0 0 0 0>
     outlineSettings {
      bUseBrushTransparency=false
      roundingType=HalfHeightRadius
      width=0
      color {colorUseRule=UseColor_Specified, specifiedColor=<rgba 0 0 0 0>}
      cornerRadii=<0 0 0 0>
     }
     tintColor {colorUseRule=UseColor_Specified, specifiedColor=<rgba 1 1 1 1>}
     uVRegion {bIsValid=false, max=<0 0>, min=<0 0>}
    }}
   slot {class=HorizontalBoxSlot, properties {horizontalAlignment=HAlign_Center, verticalAlignment=VAlign_Center, padding=<ltrb 0 0 10 0>}}
  }
  DiamondStatePip {
   widgetClassPath=/FortniteCoreUI/UI/DiscoveryShared/Status/DiamondStatePip
   bIsVariable=true
   parent=Border_PipColor
   widget=DiamondStatePip
   slot {class=BorderSlot}
  }
  Image_Dim {
   widgetClassPath=&Image
   bIsVariable=true
   parent=Overlay_21
   widget=Image_Dim
   properties {
    renderOpacity=0
    visibility=SelfHitTestInvisible
    brush {
     bIsDynamicallyLoaded=false
     drawAs=Image
     imageType=NoImage
     mirroring=NoMirror
     resourceName=None
     resourceObject=null
     tiling=NoTile
     imageSize=<32 32>
     margin=<ltrb 0 0 0 0>
     outlineSettings {
      bUseBrushTransparency=false
      roundingType=HalfHeightRadius
      width=0
      color {colorUseRule=UseColor_Specified, specifiedColor=<rgba 0 0 0 0>}
      cornerRadii=<0 0 0 0>
     }
     tintColor {colorUseRule=UseColor_Specified, specifiedColor=<rgba 1 1 1 0.25>}
     uVRegion {bIsValid=false, max=<0 0>, min=<0 0>}
    }
    colorAndOpacity=<rgba 0 0.0056 0.1329 1>
   }
   slot {class=OverlaySlot, properties {horizontalAlignment=HAlign_Fill, verticalAlignment=VAlign_Fill}}
  }
 ]
}
