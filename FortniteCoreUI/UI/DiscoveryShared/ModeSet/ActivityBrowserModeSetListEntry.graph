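# Tile widget for one entry in the Discovery activity browser's mode-set list, built around a CategoryButton hosting a key-art thumbnail.
# Hover or non-mouse selection drives thumbnail zoom and button hover state; saved state drives button selection.
# Loaded preview images fade in through animated dynamic-material parameters.
# Name, party-size and saved-state events recheck validity, setting the button's localized warning and animating a disabled overlay on validity changes.
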WidgetBlueprint /FortniteCoreUI/UI/DiscoveryShared/ModeSet/ActivityBrowserModeSetListEntry
animations [
 OnImageLoaded {
  binding_count=1
  length_seconds=0.24
  object=&self:OnImageLoaded
  sequence {
   bindings [Thumbnail {
     id=#1
     possessed_class=Image
     tracks [{
       type=MovieSceneWidgetMaterialTrack
       sections [{
         type=MovieSceneParameterSection
         end_frame=24
         start_frame=0
         channels [
          Thumbnail Alpha {type=MovieSceneScriptingFloatChannel, keys [[0,0],[24,1]]}
          Thumbnail Zoom {type=MovieSceneScriptingFloatChannel, keys [[0,1.05],[24,1]]}
         ]
        }]
      }]
    }]
   display_rate {numerator=100}
  }
 }
 OnHoverUnhoverRehover {
  binding_count=1
  length_seconds=0.16
  object=&self:OnHoverUnhoverRehover
  sequence {bindings [Thumbnail {
     id=#2
     possessed_class=Image
     tracks [{
       type=MovieSceneWidgetMaterialTrack
       sections [{
         type=MovieSceneParameterSection
         end_frame=3
         start_frame=0
         channels [Thumbnail Zoom {type=MovieSceneScriptingFloatChannel, keys [[0,1],[3.2,1.05]]}]
        }]
      }]
    }]}
 }
 SetDisabledColoration {
  binding_count=1
  length_seconds=0.16
  object=&self:SetDisabledColoration
  sequence {
   bindings [Thumbnail {
     id=#3
     possessed_class=Image
     tracks [{
       type=MovieSceneWidgetMaterialTrack
       sections [{
         type=MovieSceneParameterSection
         end_frame=16
         start_frame=0
         channels [Overlay Opacity {type=MovieSceneScriptingFloatChannel, keys [[0,0],[16,1]]}]
        }]
      }]
    }]
   display_rate {numerator=100}
  }
 }
]
graphs [
 EventGraph {nodes [
   n1 = SetHoveredState!() -> (OutputDelegate, then, IsHovered:bool=~) {is=title K2Node_CustomEvent}
   > Branch(Condition="Not Equal (Boolean)"(A=@n1.IsHovered, B=Get IsInHoveredState() -> (IsInHoveredState:bool)) -> (ReturnValue=~))
   > n2 = Set IsInHoveredState(IsInHoveredState:bool=@n1.IsHovered) -> (Output_Get:bool)
   > n3 = Branch(Condition=@n2)
   n4 = Get OnHoverUnhoverRehover() -> (OnHoverUnhoverRehover:Widget Animation Object Reference)
   > PlayAnimationForward(InAnimation=@n4)
   n5 = Get CategoryButton() -> (CategoryButton:Category Button Object Reference)
   > SetHovered!(execute, self:Category Button Object Reference=@n5, IsHovered:bool=true) -> (then)
   PlayAnimationReverse(execute=@n3.else, InAnimation=@n4)
   > SetHovered(self=@n5, IsHovered=false)
   Event On Hovered()
   > SetHoveredState!(execute, self, IsHovered:bool=true) -> (then) {is=title K2Node_CallFunction}
   n6 = Event OnSavedActivitySet!() -> (OutputDelegate, then, IsSaved:bool) {is=title K2Node_Event}
   > RefreshDisabledState()
   > n7 = Branch(Condition=@n6.IsSaved)
   n8 = Get CategoryButton()
   > SetSelected!(execute, self:Category Button Object Reference=@n8, IsSelected:bool=true) -> (then)
   SetSelected(execute=@n7.else, self=@n8, IsSelected=false)
   n9 = Event On Item Selection Changed!() -> (OutputDelegate, then, bIsSelected:bool) {is=title K2Node_Event}
   n10 = Get CommonInputSubsystem()
   > Branch(Condition="Not Equal (Enum)"!(A:ECommonInputType Enum=GetCurrentInputType(self=@n10), B:ECommonInputType Enum=MouseAndKeyboard) -> (ReturnValue:bool))
   > SetHoveredState(IsHovered=@n9.bIsSelected)
   n11 = Event OnActivityNameSet!() -> (OutputDelegate, then, ActivityName:"Text (by ref)") {is=title K2Node_Event}
   > SetTitleText!(execute, self:Category Button Object Reference=Get CategoryButton(), TitleText:"Text (by ref)"=@n11.ActivityName) -> (then)
   > RefreshDisabledState()
   Event Construct()
   > Parent: Construct()
   > n12 = GetDynamicMaterial(self=Get Thumbnail() -> (Thumbnail:Image Object Reference))
   > SetScalarParameterValue(self=@n12, ParameterName=Get ThumbnailAlphaParam() -> (ThumbnailAlphaParam:Name=None), Value=0)
   > n13 = Get Texture Parameter Value!(
    execute
    self:Material Instance Dynamic Object Reference=@n12
    ParameterName:Name=Get ThumbnailImageParam() -> (ThumbnailImageParam:Name=None)
   ) -> (then, ReturnValue:Texture Object Reference) {is=title K2Node_CallFunction}
   > Set DefaultImage(DefaultImage:Texture Object Reference=@n13) -> (Output_Get:Texture Object Reference)
   n14 = Event On Unhovered()
   SetHoveredState(execute=[@n14.then, @n15.then], IsHovered=false)
   Event On Clicked()
   > PlayClickedEvents(self:Category Button Object Reference=Get CategoryButton())
   Event OnPartySizeChanged!() -> (OutputDelegate, then, PartySize:int=0) {is=title K2Node_Event}
   > RefreshDisabledState()
   Event PreConstruct()
   > SetSafeZonePadding!(
    execute
    self:Category Button Object Reference=Get CategoryButton()
    Pad Left:bool=Get Pad Left() -> (Pad Left:bool)
    Pad Right:bool=Get Pad Right() -> (Pad Right:bool)
   ) -> (then)
   n15 = Event OnFocusLost!() -> (OutputDelegate, then, InFocusEvent:Focus Event Structure) {is=title K2Node_Event}
   n16 = Event OnPreviewImageChanged!() -> (OutputDelegate, then, bIsLoading:bool, Texture:Texture Object Reference) {is=title K2Node_Event}
   > n17 = Branch(Condition=NOT Boolean(A=@n16.bIsLoading))
   > n18 = Is Valid(InputObject=@n16.Texture)
   n19 = GetDynamicMaterial(execute=[@n18.Is Not Valid, @n17.else], self=Get Thumbnail())
   > SetScalarParameterValue(self=@n19, ParameterName=Get ThumbnailAlphaParam() -> (ThumbnailAlphaParam=None), Value=0)
   > Get Texture Parameter Value(self=@n19, ParameterName=Get ThumbnailImageParam() -> (ThumbnailImageParam=None))
   > Set IsKeyArtValid(IsKeyArtValid:bool=false) -> (Output_Get:bool)
   n20 = GetDynamicMaterial(execute=@n18.Is Valid, self=Get Thumbnail())
   > SetTextureParameterValue(self=@n20, ParameterName=Get ThumbnailImageParam() -> (ThumbnailImageParam=None), Value=@n16.Texture)
   > PlayAnimation(InAnimation=Get OnImageLoaded() -> (OnImageLoaded:Widget Animation Object Reference))
   > Set IsKeyArtValid(IsKeyArtValid=true)
  ]}
 RefreshDisabledState {nodes [
   RefreshDisabledState() {is=title K2Node_FunctionEntry}
   > Set IsActivityPreviouslyValid(IsActivityPreviouslyValid:bool=Get IsActivityValid() -> (IsActivityValid:bool)) -> (Output_Get:bool)
   n1 = GetInvalidActivityReason!(self:Fort Activity Browser Mode Set List Entry Object Reference) -> (ReturnValue:EFortInvalidActivityReason Enum=None)
   > n2 = Set IsActivityValid(IsActivityValid:bool="Equal (Enum)"(A:EFortInvalidActivityReason Enum=@n1, B:EFortInvalidActivityReason Enum=None)) -> (Output_Get:bool)
   > SetWarningStatus!(
    execute
    self:Category Button Object Reference=Get CategoryButton()
    WarningOn:bool=NOT Boolean(A=@n2)
    WarningMessage:Text=Select!(
     None:Text
     PartyTooBig:Text="LOCTABLE(\"/Game/UI/DiscoveryShared/StringTable_ActivityBrowser\", \"Warning_PartyTooLarge\")"
     NotPartyLeader:Text="LOCTABLE(\"&StringTable_ActivityBrowser\", \"Warning_NotPartyLeader\")"
     MatchmakingAlready:Text="LOCTABLE(\"&StringTable_ActivityBrowser\", \"Warning_CurrentlyMatchmaking\")"
     PlaylistDisabled:Text="LOCTABLE(\"&StringTable_ActivityBrowser\", \"Warning_ModeDisabled\")"
     PlayerIsLfg:Text="LOCTABLE(\"&StringTable_ActivityBrowser\", \"Warning_PlayerIsLfg\")"
     InvalidData:Text="LOCTABLE(\"&StringTable_ActivityBrowser\", \"Warning_InvalidMode\")"
     ExpiredTournamentLockCache:Text
     Index:EFortInvalidActivityReason Enum=@n1
    ) -> (ReturnValue:Text)
   ) -> (then)
   > Branch(Condition="Not Equal (Boolean)"(A=Get IsActivityValid(), B=Get IsActivityPreviouslyValid() -> (IsActivityPreviouslyValid:bool)) -> (ReturnValue=~))
   > n3 = Branch(Condition=Get IsActivityValid())
   n4 = Get SetDisabledColoration() -> (SetDisabledColoration:Widget Animation Object Reference)
   > PlayAnimationReverse(InAnimation=@n4)
   PlayAnimationForward(execute=@n3.else, InAnimation=@n4)
  ]}
 ResetSetDisabledColorationAnimation {nodes [
   ResetSetDisabledColorationAnimation() {is=title K2Node_FunctionEntry}
   > n1 = GetDynamicMaterial(self=Get Thumbnail())
   > SetScalarParameterValue(self=@n1, ParameterName=Get ThumbnailOverlayOpacityParam() -> (ThumbnailOverlayOpacityParam:Name=None), Value=0)
  ]}
]
variables [
 DefaultImage {type=Texture, default=null}
 DisabledClickDuration {type=number, default=2.5}
 IsActivityValid {type=boolean, default=true}
 IsInHoveredState {type=boolean}
 IsKeyArtValid {type=boolean}
 Pad Left {type=boolean}
 Pad Right {type=boolean}
 ThumbnailAlphaParam {type=string, default=Thumbnail Alpha}
 ThumbnailImageParam {type=string, default=Thumbnail}
 ThumbnailOverlayOpacityParam {type=string, default=Overlay Opacity}
 TileSize {type=Vector2D, default=<384 224>}
]
widgets {
 info {
  inheritedWidgetCount=1
  namedSlotCount=1
  parentClass=/Script/FortniteUI.FortActivityBrowserModeSetListEntry
  rootWidgetClass=&/CategoryButton.CategoryButton_C
  widgetCount=9
 }
 widgets [
  CategoryButton {
   widgetClassPath=&/CategoryButton
   bIsVariable=true
   widget=CategoryButton
   properties {
    bDisplayInputActionWhenNotInteractable=false
    visibility=HitTestInvisible
    triggeringInputAction {dataTable=/FortUILibrary/Foundation/ActivityBrowserInputActionTable, rowName=SubModeSelect}
   }
  }
  SizeBox_Primary {
   widgetClassPath=&SizeBox
   bIsVariable=true
   namedSlotHost=CategoryButton
   widget=SizeBox_Primary
   properties {heightOverride=224, widthOverride=384}
  }
  Ovr_Primary {widgetClassPath=&Overlay, bIsVariable=true, parent=SizeBox_Primary, widget=Ovr_Primary, slot {class=SizeBoxSlot}}
  Thumbnail {
   widgetClassPath=&Image
   bIsVariable=true
   parent=Ovr_Primary
   widget=Thumbnail
   properties {
    visibility=HitTestInvisible
    brush {
     bIsDynamicallyLoaded=false
     drawAs=Box
     imageType=NoImage
     mirroring=NoMirror
     resourceName=None
     resourceObject=/FortniteCoreUI/UI/DiscoveryShared/Art/MaterialInstances/MI_UI_Discovery_Tile_Thumbnail_ModeSelect
     tiling=NoTile
     imageSize=<128 128>
     margin=<ltrb 0.5 0.5 0.5 0.5>
     outlineSettings {
      bUseBrushTransparency=false
      roundingType=HalfHeightRadius
      width=0
      color {colorUseRule=UseColor_Specified, specifiedColor=<rgba 0 0 0 0>}
      cornerRadii=<0 0 0 0>
     }
     tintColor {colorUseRule=UseColor_Specified, specifiedColor=<rgba 1 1 1 1>}
     uVRegion {bIsValid=false, max=<0 0>, min=<0 0>}
    }
   }
   slot {class=OverlaySlot, properties {horizontalAlignment=HAlign_Fill, verticalAlignment=VAlign_Fill}}
  }
  Overlay_Interior {
   widgetClassPath=&Overlay
   parent=Ovr_Primary
   widget=Overlay_Interior
   slot {class=OverlaySlot, properties {horizontalAlignment=HAlign_Fill, verticalAlignment=VAlign_Fill}}
  }
  OvrDisabledWarning {
   widgetClassPath=&Overlay
   bIsVariable=true
   parent=Overlay_Interior
   widget=OvrDisabledWarning
   properties {visibility=Collapsed}
   slot {class=OverlaySlot, properties {horizontalAlignment=HAlign_Fill, verticalAlignment=VAlign_Fill}}
  }
  SizeBox_1 {
   widgetClassPath=&SizeBox
   parent=OvrDisabledWarning
   widget=SizeBox_1
   properties {bOverride_HeightOverride=true, heightOverride=128}
   slot {class=OverlaySlot, properties {horizontalAlignment=HAlign_Fill, padding=<ltrb 0 20 0 0>}}
  }
  DisabledWarningMessage {
   widgetClassPath=&CommonTextBlock
   bIsVariable=true
   parent=SizeBox_1
   widget=DisabledWarningMessage
   properties {
    justification=Center
    renderOpacity=0
    text="{Disabled warning message. Lorem ipsum.}"
    textTransformPolicy=ToUpper
    visibility=Collapsed
    wrapTextAt=240
    colorAndOpacity {colorUseRule=UseColor_Specified, specifiedColor=<rgba 0.9301 1 0.0097 1>}
    font {
     bForceMonospaced=false
     bMaterialIsStencil=false
     fontMaterial=null
     fontObject=&BurbankSmall
     letterSpacing=0
     monospacedWidth=1
     size=14
     skewAmount=0
     typefaceFontName=Black
     outlineSettings {
      bApplyOutlineToDropShadows=false
      bMiteredCorners=false
      bSeparateFillAlpha=false
      outlineMaterial=null
      outlineSize=0
      outlineColor=<rgba 0 0 0 1>
     }
    }
    shadowOffset=<0 0>
   }
   slot {class=SizeBoxSlot, properties {horizontalAlignment=HAlign_Center, verticalAlignment=VAlign_Center, padding=<ltrb 64 0 64 0>}}
  }
  InputActionWidget {widgetClassPath=&CommonActionWidget, bInherited=true, widget=null}
 ]
}
